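# Physics asset for the SatireCane Ode male medium player body cosmetic, covering a sword bone chain and a few body bones.
# A small mstr_sword sphere anchors a constrained chain to dyn_sword and dyn_sword01, with narrow swing and twist angles suggesting restrained secondary sway.
# Capsules on pelvis, spine_03 and thigh_l give coarse body shapes, likely for the sword to collide against.
# Only thigh_l is constrained to pelvis, with free angular motion; spine_03 has no constraint.
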PhysicsAsset /BRCosmetics/Characters/Player/Male/Medium/Bodies/M_MED_SatireCane_OdeYap/Meshes/M_MED_SatireCane_Ode_Physics
bodies {
 dyn_sword [{shapeType=Box, extentX=6.3025, extentY=4.8683, extentZ=1.01, shapeName=dyn_sword_box, center=<2.0402 0 0>, rotation=<rot 0 -90 0>}]
 dyn_sword01 [
  {
   shapeType=Capsule
   length=12.2347
   radius=5.7823
   shapeName=dyn_sword01_capsule
   center=<-10.8518 4.2537 6.0583>
   rotation=<rot -88.6513 157.4254 -173.8315>
  }
  {
   shapeType=Box
   extentX=9.9453
   extentY=4.2655
   extentZ=34.0499
   shapeName=dyn_sword01_box
   center=<12.2092 -1.01 2.508>
   rotation=<rot -0.0065 83.0704 -91.5255>
  }
 ]
 mstr_sword [{shapeType=Sphere, radius=0.505, shapeName=mstr_sword_sphere}]
 pelvis [
  {shapeType=Capsule, length=6.9973, radius=10.7567, shapeName=pelvis_capsule, center=<-1.0732 -2.5797 0>}
  {shapeType=Capsule, length=5.3847, radius=11.0436, shapeName=pelvis_capsule, center=<9.1657 -0.8442 0>}
 ]
 spine_03 [{
   shapeType=Capsule
   length=3.1032
   radius=11.4245
   shapeName=spine_03_capsule
   center=<1.8817 1.5357 -0.3809>
   rotation=<rot -0.0873 -90 -0.1006>
  }]
 thigh_l [{shapeType=Capsule, length=24.1823, radius=8.0989, shapeName=thigh_l_capsule, center=<-12.1236 0 0.6263>, rotation=<rot -90 -0.0001 0>}]
}
constraints [
 {
  bone1Name=thigh_l
  bone2Name=pelvis
  swing1LimitDegrees=45
  swing1Motion=Free
  swing2LimitDegrees=45
  swing2Motion=Free
  twistLimitDegrees=45
  twistMotion=Free
 }
 {bone1Name=dyn_sword, bone2Name=mstr_sword, swing1LimitDegrees=20, swing2LimitDegrees=20, twistLimitDegrees=5}
 {bone1Name=dyn_sword01, bone2Name=dyn_sword, swing1LimitDegrees=15, swing2LimitDegrees=15, twistLimitDegrees=5}
]
